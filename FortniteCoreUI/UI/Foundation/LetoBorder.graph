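# LetoBorder is a layout border widget on FortLetoLayoutBorder that frames a screen region and shows a player nameplate.
# Four dark translucent edge Borders surround a hit-test-invisible center outline, with a bottom-center nameplate holding the inherited player name text.
# UpdateBorderWidth sets the padding of all four edges to HalfBorderWidth, which sets their thickness.
# Each edge wraps a SafeZone padding only its own side, so the frame widens with the safe area.
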WidgetBlueprint /FortniteCoreUI/UI/Foundation/LetoBorder
graphs [EventGraph {nodes [
   n1 = Event UpdateBorderWidth!() -> (OutputDelegate, then, HalfBorderWidth:float) {is=title K2Node_Event}
   > SetPadding!(
    execute
    self:Border Object Reference=[
     Get BorderBottom() -> (BorderBottom:Border Object Reference)
     Get BorderLeft() -> (BorderLeft:Border Object Reference)
     Get BorderRight() -> (BorderRight:Border Object Reference)
     Get BorderTop() -> (BorderTop:Border Object Reference)
    ]
    InPadding:Margin Structure=Make Margin(Left=@n1.HalfBorderWidth, Top=@n1.HalfBorderWidth, Right=@n1.HalfBorderWidth, Bottom=@n1.HalfBorderWidth)
   ) -> (then)
  ]}]
variables []
widgets {
 info {inheritedWidgetCount=1, parentClass=/Script/FortniteUI.FortLetoLayoutBorder, rootWidgetClass=&Overlay, widgetCount=15}
 widgets [
  Overlay_0 {widgetClassPath=&Overlay, widget=Overlay_0}
  VerticalBox_102 {
   widgetClassPath=&VerticalBox
   parent=Overlay_0
   widget=VerticalBox_102
   slot {class=OverlaySlot, properties {horizontalAlignment=HAlign_Fill, verticalAlignment=VAlign_Fill}}
  }
  BorderTop {
   widgetClassPath=&Border
   bIsVariable=true
   parent=VerticalBox_102
   widget=BorderTop
   properties {
    background {
     bIsDynamicallyLoaded=false
     drawAs=Image
     imageType=NoImage
     mirroring=NoMirror
     resourceName=None
     resourceObject=null
     tiling=NoTile
     imageSize=<32 32>
     margin=<ltrb 0 0 0 0>
     outlineSettings {
      bUseBrushTransparency=false
      roundingType=HalfHeightRadius
      width=0
      color {colorUseRule=UseColor_Specified, specifiedColor=<rgba 0 0 0 0>}
      cornerRadii=<0 0 0 0>
     }
     tintColor {colorUseRule=UseColor_Specified, specifiedColor=<rgba 0 0 0 0.85>}
     uVRegion {bIsValid=false, max=<0 0>, min=<0 0>}
    }
    padding=<ltrb 18 18 18 18>
   }
   slot {class=VerticalBoxSlot}
  }
  SafeZone_188 {
   widgetClassPath=&SafeZone
   parent=BorderTop
   widget=SafeZone_188
   properties {padBottom=false, padLeft=false, padRight=false}
   slot {class=BorderSlot, properties {padding=<ltrb 18 18 18 18>}}
  }
  HorizontalBox_133 {
   widgetClassPath=&HorizontalBox
   parent=VerticalBox_102
   widget=HorizontalBox_133
   slot {class=VerticalBoxSlot, properties {size {sizeRule=Fill, value=1}}}
  }
  BorderLeft {
   widgetClassPath=&Border
   bIsVariable=true
   parent=HorizontalBox_133
   widget=BorderLeft
   properties {
    background {
     bIsDynamicallyLoaded=false
     drawAs=Image
     imageType=NoImage
     mirroring=NoMirror
     resourceName=None
     resourceObject=null
     tiling=NoTile
     imageSize=<32 32>
     margin=<ltrb 0 0 0 0>
     outlineSettings {
      bUseBrushTransparency=false
      roundingType=HalfHeightRadius
      width=0
      color {colorUseRule=UseColor_Specified, specifiedColor=<rgba 0 0 0 0>}
      cornerRadii=<0 0 0 0>
     }
     tintColor {colorUseRule=UseColor_Specified, specifiedColor=<rgba 0 0 0 0.85>}
     uVRegion {bIsValid=false, max=<0 0>, min=<0 0>}
    }
    padding=<ltrb 18 18 18 18>
   }
   slot {class=HorizontalBoxSlot}
  }
  SafeZone_1 {
   widgetClassPath=&SafeZone
   parent=BorderLeft
   widget=SafeZone_1
   properties {padBottom=false, padRight=false, padTop=false}
   slot {class=BorderSlot, properties {padding=<ltrb 18 18 18 18>}}
  }
  Border_25 {
   widgetClassPath=&Border
   parent=HorizontalBox_133
   widget=Border_25
   properties {
    visibility=HitTestInvisible
    background {
     bIsDynamicallyLoaded=false
     drawAs=Border
     imageType=NoImage
     mirroring=NoMirror
     resourceName=None
     resourceObject=/Game/Global/Textures/Default/Blanks/T_Blank_LegalWhite_sRGBColor
     tiling=NoTile
     imageSize=<1 1>
     margin=<ltrb 0.2 0.2 0.2 0.2>
     outlineSettings {
      bUseBrushTransparency=false
      roundingType=HalfHeightRadius
      width=0
      color {colorUseRule=UseColor_Specified, specifiedColor=<rgba 0 0 0 0>}
      cornerRadii=<0 0 0 0>
     }
     tintColor {colorUseRule=UseColor_Specified, specifiedColor=<rgba 1 1 1 1>}
     uVRegion {bIsValid=false, max=<0 0>, min=<0 0>}
    }
    padding=<ltrb 0 0 0 0>
   }
   slot {class=HorizontalBoxSlot, properties {size {sizeRule=Fill, value=1}}}
  }
  BorderRight {
   widgetClassPath=&Border
   bIsVariable=true
   parent=HorizontalBox_133
   widget=BorderRight
   properties {
    background {
     bIsDynamicallyLoaded=false
     drawAs=Image
     imageType=NoImage
     mirroring=NoMirror
     resourceName=None
     resourceObject=null
     tiling=NoTile
     imageSize=<32 32>
     margin=<ltrb 0 0 0 0>
     outlineSettings {
      bUseBrushTransparency=false
      roundingType=HalfHeightRadius
      width=0
      color {colorUseRule=UseColor_Specified, specifiedColor=<rgba 0 0 0 0>}
      cornerRadii=<0 0 0 0>
     }
     tintColor {colorUseRule=UseColor_Specified, specifiedColor=<rgba 0 0 0 0.85>}
     uVRegion {bIsValid=false, max=<0 0>, min=<0 0>}
    }
    padding=<ltrb 18 18 18 18>
   }
   slot {class=HorizontalBoxSlot}
  }
  SafeZone_2 {
   widgetClassPath=&SafeZone
   parent=BorderRight
   widget=SafeZone_2
   properties {padBottom=false, padLeft=false, padTop=false}
   slot {class=BorderSlot, properties {padding=<ltrb 18 18 18 18>}}
  }
  BorderBottom {
   widgetClassPath=&Border
   bIsVariable=true
   parent=VerticalBox_102
   widget=BorderBottom
   properties {
    background {
     bIsDynamicallyLoaded=false
     drawAs=Image
     imageType=NoImage
     mirroring=NoMirror
     resourceName=None
     resourceObject=null
     tiling=NoTile
     imageSize=<32 32>
     margin=<ltrb 0 0 0 0>
     outlineSettings {
      bUseBrushTransparency=false
      roundingType=HalfHeightRadius
      width=0
      color {colorUseRule=UseColor_Specified, specifiedColor=<rgba 0 0 0 0>}
      cornerRadii=<0 0 0 0>
     }
     tintColor {colorUseRule=UseColor_Specified, specifiedColor=<rgba 0 0 0 0.85>}
     uVRegion {bIsValid=false, max=<0 0>, min=<0 0>}
    }
    padding=<ltrb 18 18 18 18>
   }
   slot {class=VerticalBoxSlot}
  }
  SafeZone {
   widgetClassPath=&SafeZone
   parent=BorderBottom
   widget=SafeZone
   properties {padLeft=false, padRight=false, padTop=false}
   slot {class=BorderSlot, properties {padding=<ltrb 18 18 18 18>}}
  }
  SafeZone_0 {
   widgetClassPath=&SafeZone
   bIsVariable=true
   parent=Overlay_0
   widget=SafeZone_0
   properties {padLeft=false, padRight=false, padTop=false}
   slot {class=OverlaySlot, properties {horizontalAlignment=HAlign_Center, verticalAlignment=VAlign_Bottom, padding=<ltrb 0 0 0 2>}}
  }
  BorderNameplate {
   widgetClassPath=&Border
   bIsVariable=true
   parent=SafeZone_0
   widget=BorderNameplate
   properties {
    background {
     bIsDynamicallyLoaded=false
     drawAs=Image
     imageType=NoImage
     mirroring=NoMirror
     resourceName=None
     resourceObject=/Game/Athena/UI/Frontend/Art/M_UI_LetoPromptBackground
     tiling=NoTile
     imageSize=<32 32>
     margin=<ltrb 0 0 0 0>
     outlineSettings {
      bUseBrushTransparency=false
      roundingType=HalfHeightRadius
      width=0
      color {colorUseRule=UseColor_Specified, specifiedColor=<rgba 0 0 0 0>}
      cornerRadii=<0 0 0 0>
     }
     tintColor {colorUseRule=UseColor_Specified, specifiedColor=<rgba 1 1 1 1>}
     uVRegion {bIsValid=false, max=<0 0>, min=<0 0>}
    }
    padding=<ltrb 24 2 24 2>
   }
   slot {class=SafeZoneSlot}
  }
  Text_PlayerName {
   widgetClassPath=&CommonTextBlock
   bInherited=true
   parent=BorderNameplate
   widget=Text_PlayerName
   properties {
    text=Player 2
    textTransformPolicy=ToUpper
    font {
     bForceMonospaced=false
     bMaterialIsStencil=false
     fontMaterial=null
     fontObject=&BurbankSmall
     letterSpacing=0
     monospacedWidth=1
     size=24
     skewAmount=0
     typefaceFontName=Bold
     outlineSettings {
      bApplyOutlineToDropShadows=false
      bMiteredCorners=false
      bSeparateFillAlpha=false
      outlineMaterial=null
      outlineSize=0
      outlineColor=<rgba 0 0 0 1>
     }
    }
    shadowOffset=<0 0>
   }
   slot {class=BorderSlot, properties {padding=<ltrb 24 2 24 2>}}
  }
 ]
}
